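# Masked material for the LatchSE vehicle wheel, replacing the parent wheel function's base color with black and its emissive with animated effects.
# Rim glow blends custom colors over a background using radial UV1 pulses and a slow sine-modulated mask.
# Tire glow combines mask-shaped color with a panning gradient Fresnel, selected by vertex color red.
# The result passes through shared vehicle damage and de-rez transition functions.
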Material /VehicleCosmetics/Wheels/LatchSE/Materials/M_Wheel_LatchSE
graph {
 properties {
  bAutomaticallySetUsageInEditor=false
  bUseMaterialAttributes=true
  bUsedWithMeshDeformer=true
  bUsedWithSkeletalMesh=true
  blendMode=BLEND_Masked
  pixelDepthOffsetMode=PDOM_Legacy
  usedShadingModels=Default Lit
 }
 nodes [
  n1 = MaterialFunctionCall() {materialFunction=&MF_WheelParent_Simple}
  n2 = GetMaterialAttributes(None=@n1.MaterialAttributes) {attributeGetTypes [#1]}
  n3 = Constant3Vector()
  n4 = Subtract(A=TextureCoordinate() {coordinateIndex=1}) {constB=0.5}
  n5 = Add(
   A=Distance(A=Multiply(A=@n4) {constB=4}, B=Constant())
   B=Multiply(A=Time(), B=ScalarParameter() {defaultValue=-0.1, group=5_Special Effects, parameterName=PulseSpeed, sortPriority=39})
  )
  n6 = OneMinus(None=NamedRerouteUsage())
  n7 = Multiply(A=Power(Base=Cosine(None=@n5)), B=@n6)
  n8 = VectorParameter() {group=5_Special Effects, parameterName=CustomColor3, sortPriority=35, defaultValue=<rgba 0.78 0.49 0.01 1>}
  n9 = Multiply(A=NamedRerouteUsage(), B=Multiply(A=@n8.RGB) {constB=2})
  n10 = LinearInterpolate(
   A=@n8.RGB
   B=Multiply(
    A=ScalarParameter() {defaultValue=2, group=5_Special Effects, parameterName=AccentColorEmissiveIncrease, sortPriority=40}
    B=VectorParameter() {group=5_Special Effects, parameterName=CustomColor2, sortPriority=34, defaultValue=<rgba 1 1 1 1>}.RGB
   ) {constA=2}
   Alpha=NamedRerouteUsage()
  )
  n11 = StaticSwitchParameter(
   True=LinearInterpolate(
    A=@n10
    B=VectorParameter() {group=5_Special Effects, parameterName=CustomColor1, sortPriority=38, defaultValue=<rgba 1 0 0 1>}.RGB
    Alpha=NamedRerouteUsage()
   )
   False=@n10
  ) {defaultValue=true, group=5_Special Effects, parameterName=UseTertiaryColor, sortPriority=37}
  n12 = Frac(None=@n6)
  n13 = Multiply(A=@n12, B=@n12)
  n14 = TextureSampleParameter2D() {
   group=5_Special Effects
   parameterName=Mask
   samplerType=SAMPLERTYPE_LinearColor
   sortPriority=33
   texture=/VehicleCosmetics/Wheels/LatchSE/Textures/T_Rim_LatchSE_Mask
  }
  n15 = Multiply(A=Distance(A=@n12, B=OneMinus(None=@n13)), B=ConstantBiasScale(None=Multiply(A=@n14.R, B=@n14.R)) {bias=0.001, scale=2})
  n16 = Multiply(A=Time(), B=Constant() {r=0.2})
  n17 = Divide(A=Add(A=Sine(None=@n16)) {constB=2}) {constB=3}
  n18 = LinearInterpolate(Alpha=Add(A=@n7, B=Multiply(A=@n15, B=@n17))) {constA=1, constB=0}
  n19 = VectorParameter() {group=5_Special Effects, parameterName=BackgroundColor, sortPriority=36, defaultValue=<rgba 0 0 0 1>}
  n20 = Clamp(None=Multiply(A=@n14.G, B=@n14.G))
  n21 = Clamp(None=OneMinus(None=ConstantBiasScale(None=@n20) {bias=-0.1, scale=7.5}))
  n22 = LinearInterpolate(A=@n21, B=@n19.RGB, Alpha=OneMinus(None=@n21))
  n23 = TextureSample(UVs=Panner(Coordinate=TextureCoordinate() {uTiling=0.25}) {speedX=0.5}) {texture=/VehicleCosmetics/Wheels/_SharedTextures/FX/Gradient_D}
  n24 = Transform(None=NamedRerouteUsage())
  n25 = Clamp(None=Multiply(A=Multiply(A=@n23.RGB) {constB=2}, B=Fresnel(Normal=@n24) {exponent=4})) {maxDefault=6}
  n26 = LinearInterpolate(
   A=Add(A=Multiply(A=@n7, B=@n9), B=LinearInterpolate(A=@n11, B=@n19.RGB, Alpha=@n18))
   B=Multiply(A=@n8.RGB, B=Add(A=@n22, B=@n25))
   Alpha=VertexColor().R
  )
  MaterialFunctionCall(In=MaterialFunctionCall(Material=SetMaterialAttributes!(MaterialAttributes=@n2.MaterialAttributes, Base Color=@n3, Emissive Color=@n26) {attributeSetTypes [#2, #3]}) {materialFunction=&MF_BaseVehicle_Damage}.Result) {materialFunction=&MF_VehicleDeRez_Transition}
  n27 = MaterialFunctionCall(Color=@n1.RimRGBAMask) {materialFunction=&SplitComponents}
  NamedRerouteDeclaration(None=@n27.G) {name=RimGMask, nodeColor=<rgba 1 0.3047 0 1>}
  NamedRerouteDeclaration(None=Multiply(A=@n2.Normal, B=VertexColor().R)) {name=TireNormalOutput, nodeColor=<rgba 0 1 0.25 1>}
  NamedRerouteDeclaration(None=@n27.R) {name=RimRMask, nodeColor=<rgba 0.7109 1 0 1>}
  NamedRerouteDeclaration(None=@n27.B) {name=RimBMask, nodeColor=<rgba 0 1 0.6484 1>}
  NamedRerouteDeclaration(None=ComponentMask(None=@n1.RimRGBAMask) {a=true}) {name=RimAMask, nodeColor=<rgba 1 0.3281 0 1>}
  VectorParameter() {group=5_Special Effects, parameterName=AccentColor, sortPriority=34, defaultValue=<rgba 1 1 1 1>}
  VectorParameter() {group=5_Special Effects, parameterName=CustomColor, sortPriority=35, defaultValue=<rgba 0.78 0.49 0.01 1>}
  VectorParameter() {group=5_Special Effects, parameterName=TertiaryColor, sortPriority=38, defaultValue=<rgba 1 0 0 1>}
 ]
}
